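# Sound cue for the Pouty Clap emote audio, playing an intro followed by a looping clap track.
# A mixer starts the intro wave immediately alongside a looping clap wave held back by a fixed delay of about two seconds, sequencing intro into loop.
# It uses emote attenuation, the emote music sound class and shared emote concurrency, keeps playing when silent, and sends to a music analysis submix.
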SoundCue /BRCosmetics/Sounds/Emotes/PoutyClap/Emote_PoutyClap_ClapLoop_Cue
root=Mixer_0
nodes [
 {id=Delay_0, children [WavePlayer_0], properties {delayMax=1.976, delayMin=1.976}}
 {id=Mixer_0, children [WavePlayer_2,Delay_0], properties {inputVolume [1,1]}}
 {id=WavePlayer_0, properties {bLooping=true, soundWaveAssetPtr=&/Emote_PoutyClap_ClapLoop}}
 {id=WavePlayer_2, properties {soundWaveAssetPtr=&/Emote_PoutyClap_Intro}}
]
properties {
 attenuationSettings=&Fort_Emotes_Athena_Att
 maxDistance=1600
 soundClassObject=&Fort_Music_Emote
 virtualizationMode=PlayWhenSilent
 volumeMultiplier=1
 concurrencyOverrides {maxCount=3, resolutionRule=StopOldest}
 concurrencySet [/Game/Sounds/Concurrency_Rules/Athena_EmoteFoley_Conc]
 soundSubmixSends [{minSendLevel=0, soundSubmix=&Fortnite_AudioAnalysis_Music}]
}
