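# Sound cue that plays cloth foley for the PoutyClap emote in Battle Royale cosmetics.
# A random node picks one of five equally weighted cloth wave variations.
# It uses Athena emote attenuation, the emote music sound class and a shared emote foley concurrency set, and keeps playing when silent.
# It also sends to the music audio analysis submix, despite being foley.
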SoundCue /BRCosmetics/Sounds/Emotes/PoutyClap/Emote_PoutyClap_Cloth_Cue
root=Random_0
nodes [
 {id=Random_0, children [WavePlayer_1,WavePlayer_2,WavePlayer_3,WavePlayer_4,WavePlayer_5], properties {weights [1,1,1,1,1]}}
 {id=WavePlayer_1, properties {soundWaveAssetPtr=&/Emote_PoutyClap_Cloth_01}}
 {id=WavePlayer_2, properties {soundWaveAssetPtr=&/Emote_PoutyClap_Cloth_02}}
 {id=WavePlayer_3, properties {soundWaveAssetPtr=&/Emote_PoutyClap_Cloth_03}}
 {id=WavePlayer_4, properties {soundWaveAssetPtr=&/Emote_PoutyClap_Cloth_04}}
 {id=WavePlayer_5, properties {soundWaveAssetPtr=&/Emote_PoutyClap_Cloth_05}}
]
properties {
 attenuationSettings=&Fort_Emotes_Athena_Att
 maxDistance=1600
 soundClassObject=&Fort_Music_Emote
 virtualizationMode=PlayWhenSilent
 volumeMultiplier=1
 concurrencyOverrides {maxCount=3, resolutionRule=StopOldest}
 concurrencySet [/Game/Sounds/Concurrency_Rules/Athena_EmoteFoley_Conc]
 soundSubmixSends [{minSendLevel=0, soundSubmix=&Fortnite_AudioAnalysis_Music}]
}
